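# Animated UI material for the UIKit "view more" arrow button: a white translucent chevron that bobs and emits a pulsing echo.
# A looping time value, set by Animation Speed, drives a spring-curve vertical offset scaled by viewport, with Stiffness and Damping parameters.
# The rotated MSDF chevron is sampled again as a growing copy that eases from filled to outline and fades via ping-pong, blended into the base shape's antialiased mask.
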Material /FortUILibrary/UIKit/Buttons/ArrowButtons/M_UIKit_ViewMore_Arrow
graph {
 properties {
  bAutomaticallySetUsageInEditor=false
  bUsedWithStaticMesh=false
  blendMode=BLEND_Translucent
  materialDomain=MD_UI
  usedShadingModels=Default Lit
 }
 nodes [
  n1 = Constant()
  n2 = NamedRerouteUsage()
  n3 = ConstantBiasScale(None=Sine(None=@n2))
  n4 = ScalarParameter() {defaultValue=0.1, parameterName=Stiffness}
  n5 = ScalarParameter() {parameterName=Damping}
  n6 = LinearInterpolate(Alpha=OneMinus(None=MaterialFunctionCall!("Value (0-1)"=@n3, Stiffness=@n4, Damping=@n5) {materialFunction=/FortUILibrary/MaterialFunctions/Basics/MF_UI_SpringCurve}.Spring Curve)) {constB=8}
  Add(
   A=WorldPosition().XY
   B=Multiply(
    A=AppendVector(A=@n1, B=@n6)
    B=MaterialFunctionCall() {materialFunction=/FortUILibrary/MaterialFunctions/Basics/MF_UI_ViewportScale}.Viewport Scale Factor
   )
  )
  Constant() {r=1}
  n7 = MaterialFunctionCall(UV=MaterialFunctionCall() {materialFunction=&GetUserInterfaceUV}.Normalized UV) {materialFunction=&MF_UI_UVScaling}
  n8 = Constant() {r=-0.25}
  n9 = TextureSample(
   UVs=NamedRerouteDeclaration(None=Rotator(Coordinate=@n7.Scaled to Fit, Time=@n8) {speed=6.2832}) {name=Arrow UV, nodeColor=<rgba 0 0 0 1>}
   Tex=NamedRerouteDeclaration(None=TextureObject() {samplerType=SAMPLERTYPE_LinearColor, texture=/FortUILibrary/Textures/Icons/T_UI_Icon_UIKit_Chevron_MSDF}) {name=Arrow Object, nodeColor=<rgba 0 0 0 1>}
  )
  n10 = NamedRerouteUsage()
  n11 = Constant2Vector() {r=0.05}
  n12 = Frac(None=Multiply(A=Time(), B=ScalarParameter() {defaultValue=1, parameterName=Animation Speed}))
  n13 = Multiply(
   A=Subtract(A=Add(A=@n10, B=@n11)) {constB=0.5}
   B=LinearInterpolate(Alpha=NamedRerouteDeclaration(None=@n12) {name=Arrow Time, nodeColor=<rgba 0 0 0 1>}) {constA=1, constB=0.6125}
  ) {constB=0.66}
  n14 = MaterialFunctionCall!(MSDF RGB=TextureSample(UVs=Add(A=@n13) {constB=0.5}, Tex=NamedRerouteUsage()).RGB) {materialFunction=/FortUILibrary/MaterialFunctions/Shapes/Modifiers/MF_UI_MSDFtoSDF}
  n15 = MaterialFunctionCall(DistanceField=@n14.SDF, Threshold=Constant() {r=0.5}) {materialFunction=&MF_UI_AntiAliasing}
  n16 = MaterialFunctionCall!(DistanceField=@n14.SDF, Threshold=Constant() {r=0.5}, Thickness=Constant() {r=0.1}, Alignment=Constant() {r=1}, Smoothing) {materialFunction=/FortUILibrary/MaterialFunctions/Shapes/Modifiers/MF_UI_ShapeOutline}
  n17 = MaterialFunctionCall("Value (0-1)"=@n2) {materialFunction=&MF_UI_EaseCurves}
  n18 = NamedRerouteDeclaration(None=MaterialFunctionCall!(Value=@n17.EaseInCubic, Scale, Height) {materialFunction=/FortUILibrary/MaterialFunctions/Basics/MF_UI_PingPong}.Result) {name=Ping Pong Anim, nodeColor=<rgba 0 0 0 1>}
  LinearInterpolate(
   A=MaterialFunctionCall(
    DistanceField=MaterialFunctionCall(MSDF RGB=@n9.RGB) {materialFunction=&MF_UI_MSDFtoSDF}.SDF
    Threshold=Constant() {r=0.5}
   ) {materialFunction=&MF_UI_AntiAliasing}.Outer
   Alpha=Multiply(A=LinearInterpolate(A=@n15.Outer, B=@n16.Outline, Alpha=@n17.EaseInQuint), B=LinearInterpolate(Alpha=@n18) {constB=0.5})
  )
 ]
}
