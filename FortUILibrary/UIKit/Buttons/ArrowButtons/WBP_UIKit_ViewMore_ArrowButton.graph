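# View More arrow button for the UI kit, built on CommonButtonBase around a single material-driven arrow image.
# Hover nudges the arrow up and plays Pulse, which animates material parameters for an expanding, fading ghost arrow; unhover eases it back.
# Pressing plays the default button sound.
# A bool-to-brush function selects a 64 or 128 pixel arrow brush, while bEnableMaterialPulse is declared but unread in these graphs.
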WidgetBlueprint /FortUILibrary/UIKit/Buttons/ArrowButtons/WBP_UIKit_ViewMore_ArrowButton
animations [
 Pulse {
  binding_count=1
  length_seconds=1
  object=&self:Pulse
  sequence {bindings [Image_Arrow {
     id=#1
     possessed_class=Image
     tracks [{
       type=MovieSceneWidgetMaterialTrack
       sections [{
         type=MovieSceneParameterSection
         end_frame=20
         start_frame=0
         channels [
          Arrow Opacity {type=MovieSceneScriptingFloatChannel, const [1,3]}
          Ghost Arrow Outline Opacity {type=MovieSceneScriptingFloatChannel, keys [[0,0],[11.9997,0.3],[19.9997,0]]}
          Ghost Arrow Scale {type=MovieSceneScriptingFloatChannel, keys [[0,1],[19.9997,1.7]]}
          Ghost Arrow Shape Opacity {type=MovieSceneScriptingFloatChannel, keys [[0,0],[11.9997,0.3],[15.9993,0],[19.9997,0]]}
         ]
        }]
      }]
    }]}
 }
 Hovered {
  binding_count=1
  length_seconds=0.1667
  object=&self:Hovered
  sequence {
   bindings [Image_Arrow {
     id=#2
     possessed_class=Image
     tracks [{
       type=MovieScene2DTransformTrack
       sections [{
         type=MovieScene2DTransformSection
         end_frame=4
         start_frame=0
         channels [Translation.Y {type=MovieSceneScriptingFloatChannel, keys [[0,0],[2,-10],[4,-6]]}]
        }]
      }]
    }]
   display_rate {numerator=24}
  }
 }
 Unhovered {
  binding_count=1
  length_seconds=0.1667
  object=&self:Unhovered
  sequence {
   bindings [Image_Arrow {
     id=#3
     possessed_class=Image
     tracks [{
       type=MovieScene2DTransformTrack
       sections [{
         type=MovieScene2DTransformSection
         end_frame=4
         start_frame=0
         channels [Translation.Y {type=MovieSceneScriptingFloatChannel, keys [[0,-6],[4,0]]}]
        }]
      }]
    }]
   display_rate {numerator=24}
  }
 }
]
graphs [
 CF_BoolToSlateBrush {nodes [
   n1 = CF_BoolToSlateBrush!() -> (then, Condition:bool=~) {is=title K2Node_FunctionEntry}
   > n2 = Branch(Condition=@n1)
   > n3 = Set!(
    execute
    SlateBrush_TintColor:Slate Color Structure=(SpecifiedColor=(R=1,G=1,B=1,A=1),ColorUseRule=UseColor_Specified)
    SlateBrush_DrawAs:ESlateBrushDrawType Enum=Image
    SlateBrush_Tiling:ESlateBrushTileType Enum=NoTile
    SlateBrush_Mirroring:ESlateBrushMirrorType Enum=NoMirror
    SlateBrush_ImageSize_X:double=64
    SlateBrush_ImageSize_Y:double=64
    SlateBrush_Margin:Margin Structure=(Left=0,Top=0,Right=0,Bottom=0)
    SlateBrush_ResourceObject:Object Reference=&/MI_UIKit_ViewMore_Arrow
    SlateBrush_OutlineSettings:Slate Brush Outline Settings Structure=(CornerRadii=(X=0,Y=0,Z=0,W=0),Color=(SpecifiedColor=(R=0,G=0,B=0,A=0),ColorUseRule=UseColor_Specified),Width=0,RoundingType=HalfHeightRadius,bUseBrushTransparency=False)
   ) -> (then, Output_Get:Slate Brush Structure)
   > Return Node!(execute, SlateBrushReturn:Slate Brush Structure=@n3)
   n4 = Set(
    execute=@n2.else
    SlateBrush_TintColor=(SpecifiedColor=(R=1,G=1,B=1,A=1),ColorUseRule=UseColor_Specified)
    SlateBrush_DrawAs=Image
    SlateBrush_Tiling=NoTile
    SlateBrush_Mirroring=NoMirror
    SlateBrush_ImageSize_X=128
    SlateBrush_ImageSize_Y=128
    SlateBrush_Margin=(Left=0,Top=0,Right=0,Bottom=0)
    SlateBrush_ResourceObject=/FortUILibrary/UIKit/Buttons/MaterialInstances/Arrows/MI_UIKit_ViewMore_Arrow
    SlateBrush_OutlineSettings=(CornerRadii=(X=0,Y=0,Z=0,W=0),Color=(SpecifiedColor=(R=0,G=0,B=0,A=0),ColorUseRule=UseColor_Specified),Width=0,RoundingType=HalfHeightRadius,bUseBrushTransparency=False)
   )
   > Return Node(SlateBrushReturn=@n4)
  ]}
 EventGraph {nodes [
   Event On Unhovered()
   > PlayAnimationForward(InAnimation=Get Unhovered() -> (Unhovered:Widget Animation Object Reference))
   Event On Pressed() {is=title K2Node_Event}
   > PlaySound2D(Sound=/Game/Sounds/Fort_UI_Sounds/FrontEnd/DefaultSounds/PressedButtons/MSS_UI_Button_Pressed_Default)
   Event On Hovered()
   > PlayPulseAnimation()
   > PlayAnimationForward(InAnimation=Get Hovered() -> (Hovered:Widget Animation Object Reference))
  ]}
 PlayPulseAnimation {nodes [
   PlayPulseAnimation() {is=title K2Node_FunctionEntry}
   > PlayAnimation(InAnimation=Get Pulse() -> (Pulse:Widget Animation Object Reference))
  ]}
]
variables [bEnableMaterialPulse {type=boolean}]
widgets {
 info {inheritedWidgetCount=1, parentClass=&CommonButtonBase, rootWidgetClass=&Image, widgetCount=2}
 widgets [
  Image_Arrow {
   widgetClassPath=&Image
   bIsVariable=true
   widget=Image_Arrow
   properties {brush {
     bIsDynamicallyLoaded=false
     drawAs=Image
     imageType=FullColor
     mirroring=NoMirror
     resourceName=None
     resourceObject=&MI_UIKit_ViewMore_Arrow
     tiling=NoTile
     imageSize=<128 128>
     margin=<ltrb 0 0 0 0>
     outlineSettings {
      bUseBrushTransparency=false
      roundingType=HalfHeightRadius
      width=0
      color {colorUseRule=UseColor_Specified, specifiedColor=<rgba 0 0 0 0>}
      cornerRadii=<0 0 0 0>
     }
     tintColor {colorUseRule=UseColor_Specified, specifiedColor=<rgba 1 1 1 1>}
     uVRegion {bIsValid=false, max=<0 0>, min=<0 0>}
    }}
  }
  InputActionWidget {widgetClassPath=&CommonActionWidget, bInherited=true, widget=null}
 ]
}
